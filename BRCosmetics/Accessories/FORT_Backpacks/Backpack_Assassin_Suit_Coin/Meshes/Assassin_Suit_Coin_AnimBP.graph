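# Animation Blueprint for the Assassin Suit Coin backpack cosmetic that procedurally offsets its pack bone to follow the owner's spine and skydive state.
# Each update maps the yaw of the owner mesh's spine_05 delta relative to spine_01 into a clamped roll, and converts skydiving or floating into an alpha.
# The AnimGraph starts from the local ref pose, applying an additive half-alpha offset with that roll, then a fixed skydive offset weighted by the alpha.
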AnimBlueprint /BRCosmetics/Accessories/FORT_Backpacks/Backpack_Assassin_Suit_Coin/Meshes/Assassin_Suit_Coin_AnimBP
target_skeleton=/Game/Accessories/FORT_Backpacks/Backpack_Assassin_Suit_Coin/Meshes/Assassin_Suit_Coin_Skeleton
graphs [
 AnimGraph {nodes [
   n1 = "Transform (Modify) Bone - Bone: pack"(ComponentPose=Local To Component(LocalPose=Local Space Ref Pose()), Alpha=0.5, Translation="0.000,-3,0", Rotation=Get backpack_rotate()) {
    rotationMode=BMM_Additive
    rotationSpace=BCS_BoneSpace
    translationMode=BMM_Additive
    translationSpace=BCS_BoneSpace
    alphaBoolBlend {blendOption=Linear}
    boneToModify {boneName=pack}
    translation {y=-3}
   }
   n2 = "Transform (Modify) Bone - Bone: pack"(ComponentPose=@n1, Alpha=Get backpack_translate() -> (backpack_translate:double), Translation="0.000,-2,0", Rotation="0.000,0.000,-13") {
    rotationMode=BMM_Additive
    rotationSpace=BCS_BoneSpace
    translationMode=BMM_Additive
    translationSpace=BCS_BoneSpace
    alphaBoolBlend {blendOption=Linear}
    boneToModify {boneName=pack}
    rotation {roll=-13}
    translation {y=-2}
   }
   Output Pose(Result=Component To Local(ComponentPose=@n2)) {layerGroup=None}
   AnimGraphNode_SubInput_0 = Input Pose()
  ]}
 EventGraph {nodes [
   Event BlueprintUpdateAnimation()
   n1 = GetOwnerSkeletalMeshComponent()
   n2 = BreakRotator(InRot=BreakTransform(InTransform=GetDeltaTransformFromRefPose(self=@n1, BoneName=spine_05, BaseName=spine_01)).Rotation)
   > Set backpack_rotate(backpack_rotate=MakeRotator(Roll=MapRangeClamped(Value=@n2.Yaw, InRangeA=12, InRangeB=-30, OutRangeA=8, OutRangeB=30)))
   n3 = Get bIsSkydiveFloating()
   n4 = Get bIsSkyDiving()
   > Set backpack_translate(backpack_translate:double="To Float (Boolean)"(InBool=OR Boolean(A=@n3, B=@n4))) -> (Output_Get:double)
   TryGetPawnOwner()
  ]}
]
variables [
 backpack_rotate {type=Rotator}
 backpack_translate {type=number}
]
